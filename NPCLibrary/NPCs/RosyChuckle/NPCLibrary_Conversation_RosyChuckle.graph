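# Conversation database for the Rosy Chuckle NPC, defining its greeting and the options offered when a player talks to it.
# The entry point leads to one speech node that picks contextual vacation-themed lines by first-ever conversation, a specific worn outfit or per-match conversation count, with a default fallback.
# That node branches through link nodes to shared quest, basic service, random service and Season 39 Winterfest gift conversations, plus the NPC's own dialogue entry.
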ConversationDatabase /NPCLibrary/NPCs/RosyChuckle/NPCLibrary_Conversation_RosyChuckle
entry_points [{tag=Conversation.Entry.RosyChuckle, destinations [#1]}]
nodes [
 ConversationEntryPointNode_7 {
  id=#1
  class=ConversationEntryPointNode
  outputs [#2]
  properties {entryTag {tagName=Conversation.Entry.RosyChuckle}}
 }
 Speech_C_4 {
  id=#2
  class=Speech_C
  outputs [
   #3
   #4
   #5
   #6
   #7
  ]
  properties {
   nodeName=""
   generalConfig {
    defaultMessage=Default conversation.
    contextualMessages [
     {
      message="Leave me alone, I'm on vacation."
      contextRequirements [{requirement {class=FortControllerRequirement_IsFirstEverConversationWithNPC}}]
      priority {curve {rowName=Tandem.Speech.Priority.FirstEverInteraction}}
     }
     {
      message="No chance Ship It! will take over this year, huh? Too bad."
      contextRequirements [{requirement {
         class=FortControllerRequirement_HasCID
         properties {allowedCIds [/BRCosmetics/Athena/Items/Cosmetics/Characters/Character_CardboardCrew_Holiday]}
        }}]
      priority {curve {rowName=Tandem.Speech.Priority.OutfitReaction}}
     }
     {
      message=I'm sitting this Winterfest out. Someone else can handle it.
      contextRequirements [{requirement {class=FortControllerRequirement_CheckNPCMatchConversationCount, properties {nPCUniqueIdTag {tagName=AISpawnerData.Type.Tandem.RosyChuckle}}}}]
      priority {curve {rowName=None}}
     }
     {
      message=I haven't gone fishing in forever. It's time to change that.
      contextRequirements [{requirement {class=FortControllerRequirement_CheckNPCMatchConversationCount, properties {nPCUniqueIdTag {tagName=AISpawnerData.Type.Tandem.RosyChuckle}}}}]
      priority {curve {rowName=None}}
     }
     {
      message=I just wanna lounge on the beach for a bit. Is that so bad?
      contextRequirements [{requirement {class=FortControllerRequirement_CheckNPCMatchConversationCount, properties {nPCUniqueIdTag {tagName=AISpawnerData.Type.Tandem.RosyChuckle}}}}]
      priority {curve {rowName=None}}
     }
     {
      message="I deserve a few rounds of golf. Or, as many as I want!"
      contextRequirements [{requirement {class=FortControllerRequirement_CheckNPCMatchConversationCount, properties {nPCUniqueIdTag {tagName=AISpawnerData.Type.Tandem.RosyChuckle}}}}]
      priority {curve {rowName=None}}
     }
    ]
   }
  }
 }
 ConversationLinkNode_37 {
  id=#7
  class=ConversationLinkNode
  properties {remoteEntryTag {tagName=Conversation.Entry.Season39.WF.PerfectGift}}
 }
 ConversationLinkNode_7 {id=#6, class=ConversationLinkNode, properties {remoteEntryTag {tagName=Conversation.Entry.RandomServices}}}
 ConversationLinkNode_2 {
  id=#5
  class=ConversationLinkNode
  properties {remoteEntryTag {tagName=Conversation.Entry.RosyChuckle.Dialogue}}
 }
 ConversationLinkNode_0 {id=#4, class=ConversationLinkNode, properties {remoteEntryTag {tagName=Conversation.Entry.BasicServices}}}
 ConversationLinkNode_1 {id=#3, class=ConversationLinkNode, properties {remoteEntryTag {tagName=Conversation.Entry.NPCLibrary.Quests}}}
]
